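# Sound cue for the Smasher monster's bodyfall foley, routed to the creature foley sound class.
# A quality level switch feeds one modulated random pick per tier: five waves in the first slot, a reduced set of two in the second, with the other two slots empty.
# It uses the Smasher's shared attenuation asset and a concurrency rule named for one sound per owner, stopping the oldest.
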SoundCue /Game/Sounds/Fort_Monster_Smasher/SFX/Smasher_Bodyfall
root=QualityLevel_0
nodes [
 {id=Modulator_1, children [Random_1]}
 {id=Modulator_2, children [Random_0]}
 {id=QualityLevel_0, children [Modulator_1,Modulator_2,null,null]}
 {id=Random_0, children [WavePlayer_10, WavePlayer_13], properties {weights [1,1]}}
 {id=Random_1, children [WavePlayer_5,WavePlayer_8,WavePlayer_6,WavePlayer_9,WavePlayer_7], properties {weights [1,1,1,1,1]}}
 {id=WavePlayer_10, properties {soundWaveAssetPtr=&/Smasher_Bodyfall04}}
 {id=WavePlayer_13, properties {soundWaveAssetPtr=&/Smasher_Bodyfall05}}
 {id=WavePlayer_5, properties {soundWaveAssetPtr=&Smasher_Bodyfall04}}
 {id=WavePlayer_6, properties {soundWaveAssetPtr=&/Smasher_Bodyfall01}}
 {id=WavePlayer_7, properties {soundWaveAssetPtr=&/Smasher_Bodyfall03}}
 {id=WavePlayer_8, properties {soundWaveAssetPtr=&Smasher_Bodyfall05}}
 {id=WavePlayer_9, properties {soundWaveAssetPtr=&/Smasher_Bodyfall02}}
]
properties {
 attenuationSettings=/Game/Sounds/Attenuation/Fort_Monster_Smasher_Att
 maxDistance=7000
 soundClassObject=/Game/Sounds/FortSoundClassesAndModes/Fort_Creatures_Foley
 volumeMultiplier=1
 attenuationOverrides {distanceAlgorithm=NaturalSound, falloffDistance=5000, attenuationShapeExtents {x=1000}}
 concurrencyOverrides {resolutionRule=PreventNew}
 concurrencySet [/Game/Sounds/Concurrency_Rules/OneMaximum_Owner_StopOldest]
}
