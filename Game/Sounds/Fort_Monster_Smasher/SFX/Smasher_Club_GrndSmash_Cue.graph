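# Sound cue for the Smasher monster's club ground smash, routed to the creature foley sound class.
# A mixer layers a smash impact, a quieter effort sound and a footstep delayed by a fixed fraction of a second.
# Each layer passes through its own modulator, with the effort given a much higher pitch ceiling.
# An attenuation node overrides falloff at the root, and concurrency caps simultaneous plays per owner.
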SoundCue /Game/Sounds/Fort_Monster_Smasher/SFX/Smasher_Club_GrndSmash_Cue
root=Attenuation_2
nodes [
 {id=Attenuation_2, children [Mixer_6], properties {bOverrideAttenuation=true}}
 {id=Delay_2, children [Modulator_0], properties {delayMax=0.2, delayMin=0.2}}
 {id=Mixer_6, children [Modulator_2,Modulator_7,Delay_2], properties {inputVolume [0.8,0.5,1]}}
 {id=Modulator_0, children [WavePlayer_7]}
 {id=Modulator_2, children [WavePlayer_6], properties {pitchMax=1.1}}
 {id=Modulator_7, children [WavePlayer_12], properties {pitchMax=2}}
 {id=WavePlayer_12, properties {soundWaveAssetPtr=/Game/Sounds/Fort_Monster_Smasher/Efforts/Smasher_Sweet_08}}
 {id=WavePlayer_6, properties {soundWaveAssetPtr=&/Smasher_SmashImpact_04}}
 {id=WavePlayer_7, properties {soundWaveAssetPtr=&/smasher_step_new_01}}
]
properties {
 attenuationSettings=/Game/Sounds/Attenuation/Fort_Monster_Smasher_Att
 bOverrideAttenuation=true
 bOverrideConcurrency=true
 maxDistance=6000
 pitchMultiplier=1.3
 soundClassObject=/Game/Sounds/FortSoundClassesAndModes/Fort_Creatures_Foley
 volumeMultiplier=0.8
 attenuationOverrides {distanceAlgorithm=NaturalSound, falloffDistance=5000, attenuationShapeExtents {x=1000}}
 concurrencyOverrides {bLimitToOwner=true, maxCount=2}
}
