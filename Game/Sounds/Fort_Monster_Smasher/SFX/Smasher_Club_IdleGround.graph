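# Sound cue for the Smasher monster's club idle-on-ground foley, playing rock impact one-shots in the creature foley sound class.
# A quality-level switch feeds two parallel attenuation, modulator and random branches, with the remaining two quality slots empty.
# Each branch plays a random smasher rock-hard wave raised in pitch and lowered in volume; the second branch leaves one slot empty.
# The cue overrides attenuation with natural-sound falloff and uses prevent-new concurrency.
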SoundCue /Game/Sounds/Fort_Monster_Smasher/SFX/Smasher_Club_IdleGround
root=QualityLevel_0
nodes [
 {id=Attenuation_1, children [Modulator_1], properties {bOverrideAttenuation=true}}
 {id=Attenuation_2, children [Modulator_2], properties {bOverrideAttenuation=true}}
 {id=Modulator_1, children [Random_1], properties {pitchMax=2, pitchMin=1.5, volumeMax=0.3, volumeMin=0.2}}
 {id=Modulator_2, children [Random_2], properties {pitchMax=2, pitchMin=1.5, volumeMax=0.3, volumeMin=0.2}}
 {id=QualityLevel_0, children [Attenuation_1, Attenuation_2, null, null]}
 {id=Random_1, children [WavePlayer_4,WavePlayer_5,WavePlayer_6,WavePlayer_7], properties {weights [1,1,1,1]}}
 {id=Random_2, children [WavePlayer_8, null, WavePlayer_10, WavePlayer_11], properties {weights [1,1,1,1]}}
 {id=WavePlayer_10, properties {soundWaveAssetPtr=&/smasher_rock_hard2}}
 {id=WavePlayer_11, properties {soundWaveAssetPtr=&/smasher_rock_hard4}}
 {id=WavePlayer_4, properties {soundWaveAssetPtr=&/smasher_rock_hard1}}
 {id=WavePlayer_5, properties {soundWaveAssetPtr=&/smasher_rock_hard1b}}
 {id=WavePlayer_6, properties {soundWaveAssetPtr=&smasher_rock_hard2}}
 {id=WavePlayer_7, properties {soundWaveAssetPtr=&smasher_rock_hard4}}
 {id=WavePlayer_8, properties {soundWaveAssetPtr=&smasher_rock_hard1}}
]
properties {
 attenuationSettings=/Game/Sounds/Attenuation/Fort_Monster_Smasher_Att
 bOverrideAttenuation=true
 bOverrideConcurrency=true
 maxDistance=6000
 soundClassObject=/Game/Sounds/FortSoundClassesAndModes/Fort_Creatures_Foley
 volumeMultiplier=1
 attenuationOverrides {distanceAlgorithm=NaturalSound, falloffDistance=5000, attenuationShapeExtents {x=1000}}
 concurrencyOverrides {resolutionRule=PreventNew}
}
