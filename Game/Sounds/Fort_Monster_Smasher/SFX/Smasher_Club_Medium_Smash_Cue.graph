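# Sound cue for the Smasher monster's medium club smash, routed to creature foley with overridden attenuation and concurrency that prevents new instances at the limit.
# A quality node selects one of two three-layer mixes, leaving two slots empty.
# Both layer a pitch-modulated "Sweet" effort with a briefly delayed high impact.
# The first adds a delayed random footstep, the second a random club smash whose picker includes an empty slot.
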SoundCue /Game/Sounds/Fort_Monster_Smasher/SFX/Smasher_Club_Medium_Smash_Cue
root=QualityLevel_0
nodes [
 {id=Attenuation_1, children [Mixer_1], properties {bOverrideAttenuation=true}}
 {id=Attenuation_2, children [Mixer_2], properties {bOverrideAttenuation=true}}
 {id=Delay_1, children [Modulator_5], properties {delayMax=0.1, delayMin=0.1}}
 {id=Delay_2, children [Modulator_8], properties {delayMax=0.1, delayMin=0.05}}
 {id=Delay_3, children [Random_0], properties {delayMax=0.2, delayMin=0.2}}
 {id=Mixer_1, children [Modulator_3,Modulator_4,Delay_1], properties {inputVolume [0.7,0.4,1]}}
 {id=Mixer_2, children [Modulator_6,Modulator_7,Delay_2], properties {inputVolume [0.7,0.4,1]}}
 {id=Modulator_3, children [Delay_3], properties {pitchMax=1.1}}
 {id=Modulator_4, children [WavePlayer_9], properties {pitchMax=2}}
 {id=Modulator_5, children [WavePlayer_8]}
 {id=Modulator_6, children [Random_2], properties {pitchMax=1.1, pitchMin=0.92}}
 {id=Modulator_7, children [WavePlayer_14], properties {pitchMax=2}}
 {id=Modulator_8, children [WavePlayer_13]}
 {id=QualityLevel_0, children [Attenuation_1, Attenuation_2, null, null]}
 {id=Random_0, children [WavePlayer_0,WavePlayer_1,WavePlayer_2,WavePlayer_3,WavePlayer_4], properties {weights [1,1,1,1,1]}}
 {id=Random_2, children [WavePlayer_12, WavePlayer_11, null], properties {weights [1,1,1]}}
 {id=WavePlayer_0, properties {soundWaveAssetPtr=&/smasher_step_new_01}}
 {id=WavePlayer_1, properties {soundWaveAssetPtr=&/smasher_step_new_02}}
 {id=WavePlayer_11, properties {soundWaveAssetPtr=&/Smasher_Club_Smash02}}
 {id=WavePlayer_12, properties {soundWaveAssetPtr=&/Smasher_Club_Smash01}}
 {id=WavePlayer_13, properties {soundWaveAssetPtr=&/Smasher_SmashImpactHi_01}}
 {id=WavePlayer_14, properties {soundWaveAssetPtr=/Game/Sounds/Fort_Monster_Smasher/Efforts/Smasher_Sweet_08}}
 {id=WavePlayer_2, properties {soundWaveAssetPtr=&/smasher_step_new_03}}
 {id=WavePlayer_3, properties {soundWaveAssetPtr=&/smasher_step_new_04}}
 {id=WavePlayer_4, properties {soundWaveAssetPtr=&/smasher_step_new_05}}
 {id=WavePlayer_8, properties {soundWaveAssetPtr=&Smasher_SmashImpactHi_01}}
 {id=WavePlayer_9, properties {soundWaveAssetPtr=&Smasher_Sweet_08}}
]
properties {
 attenuationSettings=/Game/Sounds/Attenuation/Fort_Monster_Smasher_Att
 bOverrideAttenuation=true
 bOverrideConcurrency=true
 maxDistance=6000
 soundClassObject=/Game/Sounds/FortSoundClassesAndModes/Fort_Creatures_Foley
 volumeMultiplier=1
 attenuationOverrides {distanceAlgorithm=NaturalSound, falloffDistance=5000, attenuationShapeExtents {x=1000}}
 concurrencyOverrides {resolutionRule=PreventNew}
}
