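# Sound cue for the Smasher monster's club run-pound impact, played as creature foley.
# A Random node picks evenly among three hard rock-smash waves, then a Modulator adds variation to the chosen sample.
# It overrides attenuation with natural-sound falloff and distance-based low-pass filtering, and sets its own concurrency to block new instances when the voice limit is reached.
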SoundCue /Game/Sounds/Fort_Monster_Smasher/SFX/Smasher_Club_RunPound
root=Modulator_2
nodes [
 {id=Modulator_2, children [Random_2]}
 {id=Random_2, children [WavePlayer_2,WavePlayer_3,WavePlayer_4], properties {weights [1,1,1]}}
 {id=WavePlayer_2, properties {soundWaveAssetPtr=&/smasher_rock_hard4}}
 {id=WavePlayer_3, properties {soundWaveAssetPtr=&/smasher_rock_hard2}}
 {id=WavePlayer_4, properties {soundWaveAssetPtr=&/smasher_rock_hard1b}}
]
properties {
 attenuationSettings=/Game/Sounds/Attenuation/Fort_Monster_Smasher_Att
 bOverrideAttenuation=true
 bOverrideConcurrency=true
 maxDistance=6000
 soundClassObject=/Game/Sounds/FortSoundClassesAndModes/Fort_Creatures_Foley
 volumeMultiplier=0.3
 attenuationOverrides {
  bAttenuateWithLPF=true
  distanceAlgorithm=NaturalSound
  falloffDistance=5000
  lPFRadiusMax=3000
  lPFRadiusMin=1000
  attenuationShapeExtents {x=1000}
 }
 concurrencyOverrides {resolutionRule=PreventNew}
}
