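# Sound cue for the Smasher monster's collision impact, routed to the creature foley sound class.
# A mixer layers three waves at differing volumes: a low impact, a Smasher effort sound with a wide random pitch range, and a high impact after a short random delay, each through a modulator.
# The cue doubles overall pitch, overrides attenuation with natural-sound falloff, and blocks new instances when its concurrency limit is reached.
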SoundCue /Game/Sounds/Fort_Monster_Smasher/SFX/Smasher_Collide_Cue
root=Mixer_6
nodes [
 {id=Delay_2, children [Modulator_0], properties {delayMax=0.1, delayMin=0.05}}
 {id=Mixer_6, children [Modulator_2,Modulator_7,Delay_2], properties {inputVolume [1,0.6,1.3]}}
 {id=Modulator_0, children [WavePlayer_11]}
 {id=Modulator_2, children [WavePlayer_6], properties {pitchMax=1.1}}
 {id=Modulator_7, children [WavePlayer_12], properties {pitchMax=2}}
 {id=WavePlayer_11, properties {soundWaveAssetPtr=&/Smasher_SmashImpactHi_01}}
 {id=WavePlayer_12, properties {soundWaveAssetPtr=/Game/Sounds/Fort_Monster_Smasher/Efforts/Smasher_Sweet_08}}
 {id=WavePlayer_6, properties {soundWaveAssetPtr=&/Smasher_SmashImpact_04}}
]
properties {
 attenuationSettings=/Game/Sounds/Attenuation/Fort_Monster_Smasher_Att
 bOverrideAttenuation=true
 bOverrideConcurrency=true
 maxDistance=6000
 pitchMultiplier=2
 soundClassObject=/Game/Sounds/FortSoundClassesAndModes/Fort_Creatures_Foley
 volumeMultiplier=0.6
 attenuationOverrides {distanceAlgorithm=NaturalSound, falloffDistance=5000, attenuationShapeExtents {x=1000}}
 concurrencyOverrides {resolutionRule=PreventNew}
}
