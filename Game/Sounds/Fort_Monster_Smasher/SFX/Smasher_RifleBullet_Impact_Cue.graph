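# Sound cue for rifle bullets hitting the Smasher monster, built from reused Husk body, opponent bullet and flesh crunch sounds.
# A quality-level switch picks between two mixes, each layering randomized, pitch-modulated body impacts, quieter bullet impacts and a crunch routed through a delay node.
# The second mix has fewer waves and empty random slots; the last two quality slots are empty.
# Attenuation is overridden, and a concurrency override allows one instance, stopping the oldest.
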SoundCue /Game/Sounds/Fort_Monster_Smasher/SFX/Smasher_RifleBullet_Impact_Cue
root=QualityLevel_0
nodes [
 {id=Delay_1, children [Modulator_5]}
 {id=Delay_2, children [Modulator_8]}
 {id=Mixer_2, children [Modulator_3,Modulator_4], properties {inputVolume [1,0.3]}}
 {id=Mixer_3, children [Mixer_2,Delay_1], properties {inputVolume [1,0.35]}}
 {id=Mixer_4, children [Modulator_6,Modulator_7], properties {inputVolume [1,0.3]}}
 {id=Mixer_5, children [Mixer_4,Delay_2], properties {inputVolume [1,0.35]}}
 {id=Modulator_3, children [Random_5], properties {pitchMax=1.2, pitchMin=1}}
 {id=Modulator_4, children [Random_4], properties {pitchMax=1.3, pitchMin=0.9}}
 {id=Modulator_5, children [Random_3], properties {pitchMax=1.3}}
 {id=Modulator_6, children [Random_8], properties {pitchMax=1.2, pitchMin=1}}
 {id=Modulator_7, children [Random_7], properties {pitchMax=1.3, pitchMin=0.9}}
 {id=Modulator_8, children [Random_6], properties {pitchMax=1.1, pitchMin=0.9}}
 {id=QualityLevel_0, children [Mixer_3,Mixer_5,null,null]}
 {id=Random_3, children [WavePlayer_23, WavePlayer_24, WavePlayer_25, WavePlayer_26], properties {weights [1,1,1,1]}}
 {id=Random_4, children [WavePlayer_27, WavePlayer_28, WavePlayer_29, WavePlayer_30, WavePlayer_31], properties {weights [1,1,1,1,1]}}
 {
  id=Random_5
  children [WavePlayer_16, WavePlayer_17, WavePlayer_18, WavePlayer_19, WavePlayer_20, WavePlayer_21, WavePlayer_22]
  properties {weights [1,1,1,1,1,1,1]}
 }
 {id=Random_6, children [WavePlayer_39, WavePlayer_40, null, null], properties {weights [1,1,1,1]}}
 {id=Random_7, children [WavePlayer_43, WavePlayer_44, WavePlayer_45, WavePlayer_46, null], properties {weights [1,1,1,1,1]}}
 {id=Random_8, children [WavePlayer_32, WavePlayer_33, WavePlayer_34, WavePlayer_35, null, null, null], properties {weights [1,1,1,1,1,1,1]}}
 {id=WavePlayer_16, properties {soundWaveAssetPtr=/Game/Sounds/Fort_Monster_Husk/SFX/husk_body_impact_mono_10}}
 {id=WavePlayer_17, properties {soundWaveAssetPtr=/Game/Sounds/Fort_Monster_Husk/SFX/husk_body_impact_mono_09}}
 {id=WavePlayer_18, properties {soundWaveAssetPtr=/Game/Sounds/Fort_Monster_Husk/SFX/husk_body_impact_mono_08}}
 {id=WavePlayer_19, properties {soundWaveAssetPtr=/Game/Sounds/Fort_Monster_Husk/SFX/husk_body_impact_mono_05}}
 {id=WavePlayer_20, properties {soundWaveAssetPtr=/Game/Sounds/Fort_Monster_Husk/SFX/husk_body_impact_mono_01}}
 {id=WavePlayer_21, properties {soundWaveAssetPtr=/Game/Sounds/Fort_Monster_Husk/SFX/husk_body_impact_mono_13}}
 {id=WavePlayer_22, properties {soundWaveAssetPtr=/Game/Sounds/Fort_Monster_Husk/SFX/husk_body_impact_mono_12}}
 {id=WavePlayer_23, properties {soundWaveAssetPtr=/Game/Sounds/Fort_Impact_Sounds/Flesh/punch_crunch_04}}
 {id=WavePlayer_24, properties {soundWaveAssetPtr=/Game/Sounds/Fort_Impact_Sounds/Flesh/punch_crunch_01}}
 {id=WavePlayer_25, properties {soundWaveAssetPtr=/Game/Sounds/Fort_Impact_Sounds/Flesh/punch_crunch_02}}
 {id=WavePlayer_26, properties {soundWaveAssetPtr=/Game/Sounds/Fort_Impact_Sounds/Flesh/punch_crunch_03}}
 {id=WavePlayer_27, properties {soundWaveAssetPtr=/Game/Sounds/Fort_Weapons/Impacts/opponent_bullet_impact_01}}
 {id=WavePlayer_28, properties {soundWaveAssetPtr=/Game/Sounds/Fort_Weapons/Impacts/opponent_bullet_impact_02}}
 {id=WavePlayer_29, properties {soundWaveAssetPtr=/Game/Sounds/Fort_Weapons/Impacts/opponent_bullet_impact_03}}
 {id=WavePlayer_30, properties {soundWaveAssetPtr=/Game/Sounds/Fort_Weapons/Impacts/opponent_bullet_impact_04}}
 {id=WavePlayer_31, properties {soundWaveAssetPtr=/Game/Sounds/Fort_Weapons/Impacts/opponent_bullet_impact_05}}
 {id=WavePlayer_32, properties {soundWaveAssetPtr=&husk_body_impact_mono_10}}
 {id=WavePlayer_33, properties {soundWaveAssetPtr=&husk_body_impact_mono_09}}
 {id=WavePlayer_34, properties {soundWaveAssetPtr=&husk_body_impact_mono_08}}
 {id=WavePlayer_35, properties {soundWaveAssetPtr=&husk_body_impact_mono_05}}
 {id=WavePlayer_39, properties {soundWaveAssetPtr=&punch_crunch_04}}
 {id=WavePlayer_40, properties {soundWaveAssetPtr=&punch_crunch_01}}
 {id=WavePlayer_43, properties {soundWaveAssetPtr=&opponent_bullet_impact_01}}
 {id=WavePlayer_44, properties {soundWaveAssetPtr=&opponent_bullet_impact_02}}
 {id=WavePlayer_45, properties {soundWaveAssetPtr=&opponent_bullet_impact_03}}
 {id=WavePlayer_46, properties {soundWaveAssetPtr=&opponent_bullet_impact_04}}
]
properties {
 attenuationSettings=/Game/Sounds/Attenuation/Fort_Bullet_ImpactOpponent_Lng_Att
 bOverrideAttenuation=true
 bOverrideConcurrency=true
 maxDistance=6000
 priority=0.5
 soundClassObject=/Game/Sounds/FortSoundClassesAndModes/Fort_Creatures_Foley
 volumeMultiplier=1.08
 attenuationOverrides {falloffDistance=5000, attenuationShapeExtents {x=1000}}
 concurrencyOverrides {maxCount=1, resolutionRule=StopOldest}
 concurrencySet [&RifleBulletImpact_Conc]
}
