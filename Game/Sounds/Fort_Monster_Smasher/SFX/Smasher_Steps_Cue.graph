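# Sound cue for the Smasher monster's footsteps, with run breaths layered in and separate near and distant versions.
# A distance crossfade blends a close mix, pairing pitch-raised random steps with slightly delayed random breaths, into random distant-step samples farther away.
# A Doppler node wraps the output.
# Overrides add logarithmic, low-pass-filtered attenuation and per-owner concurrency that stops the quietest instance, under the creature foley sound class.
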SoundCue /Game/Sounds/Fort_Monster_Smasher/SFX/Smasher_Steps_Cue
root=Doppler_0
nodes [
 {id=Delay_1, children [Modulator_3], properties {delayMax=0.2, delayMin=0.15}}
 {
  id=DistanceCrossFade_1
  children [Mixer_1,Random_4]
  properties {crossFadeInput [
    {fadeInDistanceEnd=0, fadeInDistanceStart=0, fadeOutDistanceEnd=3500, fadeOutDistanceStart=2000, volume=1}
    {fadeInDistanceEnd=2500, fadeInDistanceStart=1000, fadeOutDistanceEnd=10000, fadeOutDistanceStart=8000, volume=1}
   ]}
 }
 {id=Doppler_0, children [DistanceCrossFade_1], properties {dopplerIntensity=3}}
 {id=Mixer_1, children [Delay_1,Modulator_2], properties {inputVolume [1,0.8]}}
 {id=Modulator_2, children [Random_0], properties {pitchMax=1.2, pitchMin=1, volumeMax=1, volumeMin=1}}
 {id=Modulator_3, children [Random_5]}
 {id=Random_0, children [WavePlayer_0,WavePlayer_1,WavePlayer_2,WavePlayer_3,WavePlayer_4], properties {weights [1,1,1,1,1]}}
 {id=Random_4, children [WavePlayer_22, WavePlayer_23, WavePlayer_24], properties {weights [1,1,1]}}
 {id=Random_5, children [WavePlayer_25, WavePlayer_26, WavePlayer_27, WavePlayer_28, WavePlayer_29], properties {weights [1,1,1,1,1]}}
 {id=WavePlayer_0, properties {soundWaveAssetPtr=&/smasher_step_new_01}}
 {id=WavePlayer_1, properties {soundWaveAssetPtr=&/smasher_step_new_02}}
 {id=WavePlayer_2, properties {soundWaveAssetPtr=&/smasher_step_new_03}}
 {id=WavePlayer_22, properties {soundWaveAssetPtr=&/smasher_distant_steps_01}}
 {id=WavePlayer_23, properties {soundWaveAssetPtr=&/smasher_distant_steps_02}}
 {id=WavePlayer_24, properties {soundWaveAssetPtr=&/smasher_distant_steps_03}}
 {id=WavePlayer_25, properties {soundWaveAssetPtr=/Game/Sounds/Fort_Monster_Smasher/Efforts/Smasher_Run_Breath01}}
 {id=WavePlayer_26, properties {soundWaveAssetPtr=/Game/Sounds/Fort_Monster_Smasher/Efforts/Smasher_Run_Breath02}}
 {id=WavePlayer_27, properties {soundWaveAssetPtr=/Game/Sounds/Fort_Monster_Smasher/Efforts/Smasher_Run_Breath03}}
 {id=WavePlayer_28, properties {soundWaveAssetPtr=/Game/Sounds/Fort_Monster_Smasher/Efforts/Smasher_Run_Breath04}}
 {id=WavePlayer_29, properties {soundWaveAssetPtr=/Game/Sounds/Fort_Monster_Smasher/Efforts/Smasher_Run_Breath05}}
 {id=WavePlayer_3, properties {soundWaveAssetPtr=&/smasher_step_new_04}}
 {id=WavePlayer_4, properties {soundWaveAssetPtr=&/smasher_step_new_05}}
]
properties {
 attenuationSettings=/Game/Sounds/Attenuation/Fort_Monster_Smasher_Att
 bOverrideAttenuation=true
 bOverrideConcurrency=true
 maxDistance=10000
 soundClassObject=/Game/Sounds/FortSoundClassesAndModes/Fort_Creatures_Foley
 volumeMultiplier=0.7
 attenuationOverrides {
  bAttenuateWithLPF=true
  distanceAlgorithm=Logarithmic
  falloffDistance=5000
  lPFFrequencyAtMax=1000
  lPFRadiusMax=2000
  lPFRadiusMin=500
  stereoSpread=5000
  attenuationShapeExtents {x=800}
 }
 concurrencyOverrides {bLimitToOwner=true, maxCount=4, resolutionRule=StopQuietest, volumeScale=0.8}
 concurrencySet [/Game/Sounds/Concurrency_Rules/TwoMaximum_Owner_StopOldest]
}
