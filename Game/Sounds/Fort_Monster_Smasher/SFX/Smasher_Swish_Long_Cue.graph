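# Sound cue for the Smasher monster's long swish, built from weapon whoosh waves and routed to the creature foley sound class.
# Each of the first two quality-level branches mixes a random heavy-metal whoosh with a random medium whoosh pitched down for weight.
# The second branch uses fewer waves and leaves empty random slots; the last two quality slots are empty.
# Natural-sound distance attenuation is overridden, and concurrency prevents new instances at the limit.
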SoundCue /Game/Sounds/Fort_Monster_Smasher/SFX/Smasher_Swish_Long_Cue
root=QualityLevel_0
nodes [
 {id=Attenuation_1, children [Mixer_1], properties {bOverrideAttenuation=true}}
 {id=Attenuation_2, children [Mixer_2], properties {bOverrideAttenuation=true}}
 {id=Mixer_1, children [Modulator_3,Modulator_2], properties {inputVolume [1,1]}}
 {id=Mixer_2, children [Modulator_5,Modulator_4], properties {inputVolume [1,1]}}
 {id=Modulator_2, children [Random_3], properties {pitchMax=0.7, pitchMin=0.6}}
 {id=Modulator_3, children [Random_2]}
 {id=Modulator_4, children [Random_5], properties {pitchMax=0.7, pitchMin=0.6}}
 {id=Modulator_5, children [Random_4]}
 {id=QualityLevel_0, children [Attenuation_1, Attenuation_2, null, null]}
 {id=Random_2, children [WavePlayer_9, WavePlayer_10, WavePlayer_11, WavePlayer_12], properties {weights [1,1,1,1]}}
 {id=Random_3, children [WavePlayer_13, WavePlayer_14, WavePlayer_15, WavePlayer_16, WavePlayer_17], properties {weights [1,1,1,1,1]}}
 {id=Random_4, children [WavePlayer_18, WavePlayer_19, WavePlayer_20, null], properties {weights [1,1,1,1]}}
 {id=Random_5, children [null, WavePlayer_23, WavePlayer_24, WavePlayer_25, null], properties {weights [1,1,1,1,1]}}
 {id=WavePlayer_10, properties {soundWaveAssetPtr=/Game/Sounds/Fort_Weapons/Whooshes/Whoosh_MetalHeavy_Stereo_01}}
 {id=WavePlayer_11, properties {soundWaveAssetPtr=/Game/Sounds/Fort_Weapons/Whooshes/Whoosh_MetalHeavy_Stereo_02}}
 {id=WavePlayer_12, properties {soundWaveAssetPtr=/Game/Sounds/Fort_Weapons/Whooshes/Whoosh_MetalHeavy_Stereo_03}}
 {id=WavePlayer_13, properties {soundWaveAssetPtr=/Game/Sounds/Fort_Weapons/Whooshes/Whoosh_Medium_Stereo_01}}
 {id=WavePlayer_14, properties {soundWaveAssetPtr=/Game/Sounds/Fort_Weapons/Whooshes/Whoosh_Medium_Stereo_02}}
 {id=WavePlayer_15, properties {soundWaveAssetPtr=/Game/Sounds/Fort_Weapons/Whooshes/Whoosh_Medium_Stereo_03}}
 {id=WavePlayer_16, properties {soundWaveAssetPtr=/Game/Sounds/Fort_Weapons/Whooshes/Whoosh_Medium_Stereo_05}}
 {id=WavePlayer_17, properties {soundWaveAssetPtr=/Game/Sounds/Fort_Weapons/Whooshes/Whoosh_Medium_Stereo_06}}
 {id=WavePlayer_18, properties {soundWaveAssetPtr=/Game/Sounds/Fort_Weapons/Whooshes/Whoosh_MetalHeavy_Stereo_04}}
 {id=WavePlayer_19, properties {soundWaveAssetPtr=&Whoosh_MetalHeavy_Stereo_01}}
 {id=WavePlayer_20, properties {soundWaveAssetPtr=&Whoosh_MetalHeavy_Stereo_02}}
 {id=WavePlayer_23, properties {soundWaveAssetPtr=&Whoosh_Medium_Stereo_02}}
 {id=WavePlayer_24, properties {soundWaveAssetPtr=&Whoosh_Medium_Stereo_03}}
 {id=WavePlayer_25, properties {soundWaveAssetPtr=&Whoosh_Medium_Stereo_05}}
 {id=WavePlayer_9, properties {soundWaveAssetPtr=&Whoosh_MetalHeavy_Stereo_04}}
]
properties {
 attenuationSettings=/Game/Sounds/Attenuation/Fort_Monster_Smasher_Att
 bOverrideAttenuation=true
 bOverrideConcurrency=true
 maxDistance=4000
 soundClassObject=/Game/Sounds/FortSoundClassesAndModes/Fort_Creatures_Foley
 volumeMultiplier=0.6
 attenuationOverrides {distanceAlgorithm=NaturalSound, falloffDistance=3000, attenuationShapeExtents {x=1000}}
 concurrencyOverrides {resolutionRule=PreventNew}
}
