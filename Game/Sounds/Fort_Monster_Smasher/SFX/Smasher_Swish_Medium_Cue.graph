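# Sound cue for the Smasher monster's medium swish, played as creature foley with the Smasher attenuation and prevent-new concurrency.
# Each variant layers a random Smasher swish with a random stereo weapon whoosh through a mixer, with modulators holding the whoosh pitch fixed.
# A quality-level root selects the variant: the first has five waves per layer, the second keeps three with empty random slots, and the last two are empty.
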SoundCue /Game/Sounds/Fort_Monster_Smasher/SFX/Smasher_Swish_Medium_Cue
root=QualityLevel_1
nodes [
 {id=Mixer_1, children [Modulator_2,Modulator_3], properties {inputVolume [1,1]}}
 {id=Mixer_2, children [Modulator_4,Modulator_5], properties {inputVolume [1,1]}}
 {id=Modulator_2, children [Random_2]}
 {id=Modulator_3, children [Random_3], properties {pitchMax=1, pitchMin=1}}
 {id=Modulator_4, children [Random_4]}
 {id=Modulator_5, children [Random_5], properties {pitchMax=1, pitchMin=1}}
 {id=QualityLevel_1, children [Mixer_1,Mixer_2,null,null]}
 {id=Random_2, children [WavePlayer_10, WavePlayer_11, WavePlayer_12, WavePlayer_13, WavePlayer_14], properties {weights [1,1,1,1,1]}}
 {id=Random_3, children [WavePlayer_15, WavePlayer_16, WavePlayer_17, WavePlayer_18, WavePlayer_19], properties {weights [1,1,1,1,1]}}
 {id=Random_4, children [WavePlayer_20, WavePlayer_21, null, null, WavePlayer_24], properties {weights [1,1,1,1,1]}}
 {id=Random_5, children [WavePlayer_25, null, null, WavePlayer_28, WavePlayer_29], properties {weights [1,1,1,1,1]}}
 {id=WavePlayer_10, properties {soundWaveAssetPtr=&/Smasher_Swish_Medium01}}
 {id=WavePlayer_11, properties {soundWaveAssetPtr=&/Smasher_Swish_Medium02}}
 {id=WavePlayer_12, properties {soundWaveAssetPtr=&/Smasher_Swish_Medium03}}
 {id=WavePlayer_13, properties {soundWaveAssetPtr=&/Smasher_Swish_Medium04}}
 {id=WavePlayer_14, properties {soundWaveAssetPtr=&/Smasher_Swish_Medium05}}
 {id=WavePlayer_15, properties {soundWaveAssetPtr=/Game/Sounds/Fort_Weapons/Whooshes/Whoosh_Medium_Stereo_01}}
 {id=WavePlayer_16, properties {soundWaveAssetPtr=/Game/Sounds/Fort_Weapons/Whooshes/Whoosh_Medium_Stereo_02}}
 {id=WavePlayer_17, properties {soundWaveAssetPtr=/Game/Sounds/Fort_Weapons/Whooshes/Whoosh_Medium_Stereo_03}}
 {id=WavePlayer_18, properties {soundWaveAssetPtr=/Game/Sounds/Fort_Weapons/Whooshes/Whoosh_Medium_Stereo_05}}
 {id=WavePlayer_19, properties {soundWaveAssetPtr=/Game/Sounds/Fort_Weapons/Whooshes/Whoosh_Medium_Stereo_06}}
 {id=WavePlayer_20, properties {soundWaveAssetPtr=&Smasher_Swish_Medium01}}
 {id=WavePlayer_21, properties {soundWaveAssetPtr=&Smasher_Swish_Medium02}}
 {id=WavePlayer_24, properties {soundWaveAssetPtr=&Smasher_Swish_Medium05}}
 {id=WavePlayer_25, properties {soundWaveAssetPtr=&Whoosh_Medium_Stereo_01}}
 {id=WavePlayer_28, properties {soundWaveAssetPtr=&Whoosh_Medium_Stereo_05}}
 {id=WavePlayer_29, properties {soundWaveAssetPtr=&Whoosh_Medium_Stereo_06}}
]
properties {
 attenuationSettings=/Game/Sounds/Attenuation/Fort_Monster_Smasher_Att
 bOverrideConcurrency=true
 maxDistance=7000
 soundClassObject=/Game/Sounds/FortSoundClassesAndModes/Fort_Creatures_Foley
 volumeMultiplier=0.5
 attenuationOverrides {distanceAlgorithm=NaturalSound, falloffDistance=5000, attenuationShapeExtents {x=1000}}
 concurrencyOverrides {resolutionRule=PreventNew}
}
