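# Sound cue that plays a randomized short swish for the Smasher monster, in the creature foley sound class.
# A quality-level switch feeds two Attenuation > Modulator > Random branches, leaving its other two inputs empty.
# The first branch picks among five swish waves, while the second keeps three of them plus two empty, equally weighted slots.
# The cue references Smasher attenuation settings and overrides concurrency to prevent new instances at the limit.
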SoundCue /Game/Sounds/Fort_Monster_Smasher/SFX/Smasher_Swish_Short_Cue
root=QualityLevel_0
nodes [
 {id=Attenuation_1, children [Modulator_1], properties {bOverrideAttenuation=true}}
 {id=Attenuation_2, children [Modulator_2], properties {bOverrideAttenuation=true}}
 {id=Modulator_1, children [Random_1]}
 {id=Modulator_2, children [Random_2]}
 {id=QualityLevel_0, children [Attenuation_1, Attenuation_2, null, null]}
 {id=Random_1, children [WavePlayer_5,WavePlayer_6,WavePlayer_7,WavePlayer_8,WavePlayer_9], properties {weights [1,1,1,1,1]}}
 {id=Random_2, children [null, WavePlayer_11, null, WavePlayer_13, WavePlayer_14], properties {weights [1,1,1,1,1]}}
 {id=WavePlayer_11, properties {soundWaveAssetPtr=&/Smasher_Swish_Short02}}
 {id=WavePlayer_13, properties {soundWaveAssetPtr=&/Smasher_Swish_Short04}}
 {id=WavePlayer_14, properties {soundWaveAssetPtr=&/Smasher_Swish_Short05}}
 {id=WavePlayer_5, properties {soundWaveAssetPtr=&/Smasher_Swish_Short01}}
 {id=WavePlayer_6, properties {soundWaveAssetPtr=&Smasher_Swish_Short02}}
 {id=WavePlayer_7, properties {soundWaveAssetPtr=&/Smasher_Swish_Short03}}
 {id=WavePlayer_8, properties {soundWaveAssetPtr=&Smasher_Swish_Short04}}
 {id=WavePlayer_9, properties {soundWaveAssetPtr=&Smasher_Swish_Short05}}
]
properties {
 attenuationSettings=/Game/Sounds/Attenuation/Fort_Monster_Smasher_Att
 bOverrideConcurrency=true
 maxDistance=7000
 soundClassObject=/Game/Sounds/FortSoundClassesAndModes/Fort_Creatures_Foley
 attenuationOverrides {distanceAlgorithm=NaturalSound, falloffDistance=5000, attenuationShapeExtents {x=1000}}
 concurrencyOverrides {resolutionRule=PreventNew}
}
